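# Niagara system for the Fox Warrior backpack cosmetic, adding blue chill sprites, light beams and a cyan glow.
# Chill sprites follow the skeletal mesh with an age-lerped offset, with burst count and alpha falling as owner speed rises and alpha also scaled by quality level.
# LightBeams spawns fading sprites from a squashed sphere, pushed from an origin point and slowed by drag.
# ParticleLight holds one persistent light particle, its brightness curved by time of day.
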NiagaraSystem /BRCosmetics/Accessories/FORT_Backpacks/Backpack_F_MED_Fox_Warrior/Effects/NS_Backpack_FoxWarrior
dependencies {usedDataInterfaces [&NiagaraDataInterfaceCurve], usedDynamicInputs [], usedModules [&SystemState]}
emitters {
 Chill {
  emitterUpdateScript [
   &EmitterState()
   &SpawnRate(
    SpawnRate=&ClampFloat(
     Float=&Divide_Float(A=5, B=&VectorLength(Vector=$Engine.Owner.Velocity))
     Max=10
    )
    Use Spawn Probability=false
   ) {enabled=false}
   &SpawnRate(Use Spawn Probability=false) {module=SpawnRate001}
   &SpawnBurst_Instantaneous(
    Spawn Count=&MakeIntFromFloat(Float=&ClampFloat(
      Float=&Divide_Float(A=3, B=&VectorLength(Vector=$Engine.Owner.Velocity))
      Max=3
     ))
    Use Spawn Probability=false
   )
  ]
  particleSpawnScript [
   /Niagara/Modules/Spawn/Initialization/InitializeParticle(
    Color=&RandomRangeLinearColor(Minimum=<rgba 0.1085 0.5519 1 1>, Maximum=<rgba 0.2456 0.5543 1 4>)
    Lifetime=&RandomRangeFloat(Minimum=2, Maximum=5)
    Module.Write Color=true
    Module.Write Lifetime=true
    Module.Write Mass=true
    Module.Write Material Random=true
    Module.Write Position=false
    Module.Write Scale=false
    Module.Write SpriteRotation=true
    Module.Write SpriteSize=true
    Position=&SimulationPosition()
    Sprite Rotation=&RandomRangeFloat(Minimum=0, Maximum=360)
    Sprite Size=&Vector2DFromFloat(Value=&RandomRangeFloat(Minimum=7, Maximum=12))
   )
   &DynamicMaterialParameters(
    Index 0 Param 2=5
    Module.Param0WriteEnabled=true
    Module.Param1WriteEnabled=true
    Module.Param2WriteEnabled=true
    Module.Param3WriteEnabled=true
   )
   &SkeletalMeshLocation(
    Bone Sampling Mode=::NewEnumerator4
    Inherit Velocity=false
    Interpolate Mesh Sampling=false
    MaskSpawnGroup=false
    Sampled Orientation Coordinate Space=::World
    Sampled Position Offset=<-4.25 -6.5 13>
    Sampled Position Offset Coordinate Space:ENiagaraMeshTransforms=::NewEnumerator2
    Skeletal Mesh=$User.SkeletalMesh
   ) {module=SkeletalMeshLocation001}
  ]
  particleUpdateScript [
   &ParticleState()
   &SkeletalMeshLocation(
    Bone Sampling Mode=::NewEnumerator4
    Inherit Velocity=false
    Interpolate Mesh Sampling=false
    MaskSpawnGroup=false
    Sampled Orientation Coordinate Space=::World
    Sampled Position Offset=&Lerp_Vector(A=<-2 -6 8>, B=<-6 -6 12>, Alpha=$Particles.NormalizedAge)
    Sampled Position Offset Coordinate Space:ENiagaraMeshTransforms=::NewEnumerator2
    Skeletal Mesh=$User.SkeletalMesh
   )
   &SubUVAnimation(SubUV Loop Count:NiagaraFloat=&RandomRangeFloat(Minimum=0.3, Maximum=0.6))
   &SpriteRotationRate(
    Delta Time:NiagaraFloat=$Particles.NormalizedAge
    Rotation Rate=&RandomRangeFloat(Minimum=-1, Maximum=1)
   )
   &ScaleSpriteSize(Scale Factor=&Vector2DFromFloat(Value=&FloatFromCurve(FloatCurve=DataInterface() {propertyValues {Curve {keys [
          {arriveTangent=0.7564, interpMode=RCIM_Cubic, leaveTangent=0.7564, tangentMode=RCTM_User, time=-0.0024, value=0.2}
          <at 0.9951 1.3 RCIM_Cubic>
         ]}}})))
   &ScaleColor(
    Scale Alpha=&Multiply_Float(
     A=&FloatFromCurve(FloatCurve=DataInterface() {propertyValues {Curve {keys [<at 0 0 RCIM_Cubic>,<at 0.5001 1 RCIM_Cubic>,<at 1 0>]}}})
     B=&Multiply_Float(
      A=&FloatFromCurve(
       FloatCurve=DataInterface() {propertyValues {Curve {keys [<at 0 1>,<at 1 0.3>]}}}
       CurveIndex=&Divide_Float(A=&VectorLength(Vector=$Engine.Owner.Velocity), B=500)
       Scale Curve=&RandomRangeFloat(Minimum=0.3, Maximum=1)
      )
      B=&ClampFloat(Float=&Multiply_Float(A=0.25, B=&MakeFloatFromInt(Int=$Engine.QualityLevel)))
     )
    )
    ScaleA=true
    ScaleRGB=false
   )
   &CameraOffset(Camera Offset Amount=&FloatFromCurve(
     FloatCurve=DataInterface() {propertyValues {Curve {keys [
         {interpMode=RCIM_Cubic, tangentMode=RCTM_User, time=0, value=0}
         {arriveTangent=1.9876, interpMode=RCIM_Cubic, leaveTangent=1.9876, tangentMode=RCTM_User, time=1, value=1}
        ]}}}
     Scale Curve=7
    ))
   &Color(Color=<rgba 1 1 1 20>) {enabled=false}
  ]
  rendererClasses [&NiagaraSpriteRendererProperties]
  renderers [{rendererClass=&NiagaraSpriteRendererProperties}]
 }
 LightBeams {
  emitterUpdateScript [
   &EmitterState()
   &SpawnRate(SpawnRate=12, Use Spawn Probability=false)
   /SetVariables(Emitter.OriginPoint:Vector3f=<-0.5 -5.7 2.4825>) {isSetParameters=true}
  ]
  particleSpawnScript [
   /Niagara/Modules/Spawn/Initialization/V2/InitializeParticle(
    Color Maximum=<rgba 0.2456 0.5543 1 0.5>
    Color Minimum=<rgba 0.1085 0.5519 1 0.1>
    Color Mode=::NewEnumerator2
    Lifetime Max=4
    Lifetime Min=2
    Lifetime Mode=::NewEnumerator1
    Position Mode=::NewEnumerator1
    Sprite Size=&Lerp_Vector2(
     A=<5 7>
     B=<7 20>
     Alpha=&Power(a=&RandomRangeFloat(Minimum=0, Maximum=1), b=3)
    )
    Sprite Size Mode=::NewEnumerator4
    UsePositionOffset=false
    Write Material Random=true
   )
   &SphereLocation(
    MaskSpawnGroup=false
    Non Uniform Scale=<1 0.5 1>
    Offset=&Add_Vector(A=$Emitter.OriginPoint, B=<-2 0 5>)
    Sphere Distribution:ENiagaraSphereDistributionMode=::NewEnumerator1
    Sphere Radius=3
    U Position:NiagaraFloat=&RandomRangeFloat(Minimum=0, Maximum=1)
    UseNonUniformScale=true
    UseOffset=true
    UseSphereOrientation=false
   )
   &DynamicMaterialParameters(
    Index 0 Param 1=&RandomRangeFloat(Minimum=-0.2, Maximum=0.2)
    Index 0 Param 2=1
    Index 0 Param 3=0.5
    Module.Param0WriteEnabled=true
    Module.Param1WriteEnabled=false
    Module.Param2WriteEnabled=false
    Module.Param3WriteEnabled=false
   )
   &AddVelocityFromPoint(
    Module.Use Falloff Distance=false
    Origin Offset=&Add_Vector(A=<1.5 2 -5>, B=$Emitter.OriginPoint)
    Velocity Strength=1
   )
  ]
  particleUpdateScript [
   &Drag(Drag=4, Use Linear Drag=true, Use Rotational Drag=false)
   &SolveForcesAndVelocity(Clamp Velocity=false, Limit Acceleration=false)
   &ParticleState()
   &ScaleColor(
    Scale Alpha=&FloatFromCurve(FloatCurve=DataInterface() {propertyValues {Curve {keys [
         {arriveTangent=6.668, interpMode=RCIM_Cubic, leaveTangent=6.668, tangentMode=RCTM_User, time=0, value=0}
         {interpMode=RCIM_Cubic, tangentMode=RCTM_User, time=0.5, value=1}
         <at 1 0>
        ]}}})
    ScaleA=true
    ScaleRGB=false
   )
  ]
  rendererClasses [&NiagaraSpriteRendererProperties]
  renderers [{rendererClass=&NiagaraSpriteRendererProperties}]
 }
 ParticleLight {
  emitterUpdateScript [
   &EmitterState()
   &SpawnBurst_Instantaneous(
    Spawn Count=1
    Spawn Probability=&MakeCustomFloatFromBool(
     Bool=&SetBoolByIntComparison(A=$Emitter.LoopCount, B=0, Comparison Type=::NewEnumerator4)
     True Float=1
     False Float=0
    )
    Use Spawn Probability=true
   )
  ]
  particleSpawnScript [&SystemLocation(MaskSpawnGroup=false, Offset=<-4 -5.5 18.5>)]
  particleUpdateScript [
   &ParticleState(Kill Particles When Lifetime Has Elapsed=false, Let Infinitely Lived Particles Die When Emitter Deactivates=true)
   &FN_Light_Attributes(
    Color=<rgba 0.0818 3.0004 5 1>
    Create Color Attribute=true
    Create Exponent Attribute=true
    Create Radius Attribute=true
    Exponent=2
    Radius=20
    Scale Color=&FloatFromCurve(
     FloatCurve=DataInterface() {propertyValues {Curve {keys [
         {interpMode=RCIM_Cubic, tangentMode=RCTM_User, time=0.0042, value=0.5}
         {arriveTangent=2.9927, interpMode=RCIM_Cubic, leaveTangent=2.9927, tangentMode=RCTM_User, time=0.3692, value=0.7}
         {arriveTangent=0.0434, interpMode=RCIM_Cubic, leaveTangent=0.0434, tangentMode=RCTM_User, time=0.6, value=1}
         <at 1.0042 0.5>
        ]}}}
     CurveIndex=&self:ReferenceTimeOfDay()
    )
   )
  ]
  rendererClasses [&NiagaraSpriteRendererProperties, &NiagaraLightRendererProperties]
  renderers [
   {rendererClass=&NiagaraSpriteRendererProperties}
   {rendererClass=&NiagaraLightRendererProperties, rendererIndex=1}
  ]
 }
}
summary {
 systemName=NS_Backpack_FoxWarrior
 emitters [
  {emitterName=Chill, rendererClasses [&NiagaraSpriteRendererProperties]}
  {emitterName=LightBeams, rendererClasses [&NiagaraSpriteRendererProperties]}
  {emitterName=ParticleLight, rendererClasses [&NiagaraSpriteRendererProperties, &NiagaraLightRendererProperties]}
 ]
 userVariables [
  User.bIsFrontEndPreview {type=NiagaraBool}
  User.bIsFullyDeployed {type=NiagaraBool}
  User.InteriorColor {type=LinearColor, description="", defaultValue {struct=LinearColor, value=<rgba 1 1 1 1>}}
  User.PlayerSpeed {type=NiagaraFloat, defaultValue=0, description=""}
  User.SkeletalMesh {
   type=NiagaraDataInterfaceSkeletalMesh
   defaultValue {
    struct=NiagaraExt_VariableValue_DataInterface
    value {dataInterface=&self:NiagaraDataInterfaceSkeletalMesh_0}
   }
  }
 ]
}
