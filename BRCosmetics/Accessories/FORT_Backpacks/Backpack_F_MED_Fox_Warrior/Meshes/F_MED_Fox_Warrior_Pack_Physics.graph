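# Physics asset for the Fox Warrior backpack cosmetic, defining capsule bodies and joints for its dangling arrows and straps.
# Eight arrow bodies each hang from Dyn_Arrows_Master with narrow swing cones and almost no twist.
# The belt strap hangs from Dyn_Master_Strap and branches into three-link long and short strap chains, with tight, mostly asymmetric swing limits.
# No constraint links the arrow group to the strap hierarchy.
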PhysicsAsset /BRCosmetics/Accessories/FORT_Backpacks/Backpack_F_MED_Fox_Warrior/Meshes/F_MED_Fox_Warrior_Pack_Physics
bodies {
 Dyn_Arrow_1 [{shapeType=Capsule, length=2, radius=0.5, center=<0.5 0 -0.3>, rotation=<rot 90 0 0>}]
 Dyn_Arrow_2 [{shapeType=Capsule, length=2, radius=0.5, center=<0.5 0 -0.3>, rotation=<rot 90 0 0>}]
 Dyn_Arrow_3 [{shapeType=Capsule, length=2, radius=0.5, center=<0.5 0 -0.3>, rotation=<rot 90 0 0>}]
 Dyn_Arrow_4 [{shapeType=Capsule, length=2, radius=0.5, center=<0.5 0 -0.3>, rotation=<rot 90 0 0>}]
 Dyn_Arrow_5 [{shapeType=Capsule, length=2, radius=0.5, center=<0.5 0 -0.3>, rotation=<rot 90 0 0>}]
 Dyn_Arrow_6 [{shapeType=Capsule, length=2, radius=0.5, center=<0.5 0 -0.3>, rotation=<rot 90 0 0>}]
 Dyn_Arrow_7 [{shapeType=Capsule, length=2, radius=0.5, center=<0.5 0 -0.3>, rotation=<rot 90 0 0>}]
 Dyn_Arrow_8 [{shapeType=Capsule, length=2, radius=0.5, center=<0.5 0 -0.3>, rotation=<rot 90 0 0>}]
 Dyn_Arrows_Master [{shapeType=Capsule, length=0.505, radius=0.505, rotation=<rot 0 -90 0>}]
 Dyn_Belt_Strap [{shapeType=Capsule, length=1, radius=1, center=<0 -0.79 -0.0001>}]
 Dyn_LongStrap1 [{shapeType=Capsule, length=1, radius=1, center=<1 0 0>, rotation=<rot -90 0 0>}]
 Dyn_LongStrap2 [
  {shapeType=Capsule, length=5, radius=2, center=<0.2742 -2.7288 -2.0054>, rotation=<rot -90 -1.7899 -2.6074>}
  {shapeType=Capsule, length=5, radius=1, center=<0.3 5.0235 -0.8031>, rotation=<rot -81.0077 162.5851 179.9999>}
 ]
 Dyn_LongStrap3 [{shapeType=Capsule, length=2, radius=1, center=<0.5953 0.3872 -0.3403>, rotation=<rot 0.0001 -10.3999 84.8748>}]
 Dyn_Master_Strap [{shapeType=Capsule, length=14, radius=4.5, center=<13.3418 5.9614 2.3025>, rotation=<rot -85.2365 -14.1331 19.6473>}]
 Dyn_ShortStrap_1 [{shapeType=Capsule, length=2, radius=0.5, center=<1 0 0>, rotation=<rot -90 0 0>}]
 Dyn_ShortStrap_2 [{shapeType=Capsule, length=2, radius=0.5, center=<1 0 0>, rotation=<rot -90 0 0>}]
 Dyn_ShortStrap_3 [{shapeType=Capsule, length=2, radius=0.5, center=<1 0 0>, rotation=<rot -90 0 0>}]
}
constraints [
 {bone1Name=Dyn_Arrow_1, bone2Name=Dyn_Arrows_Master, swing1LimitDegrees=10, swing2LimitDegrees=10, twistLimitDegrees=0.1}
 {bone1Name=Dyn_Arrow_2, bone2Name=Dyn_Arrows_Master, swing1LimitDegrees=10, swing2LimitDegrees=10, twistLimitDegrees=0.1}
 {bone1Name=Dyn_Arrow_3, bone2Name=Dyn_Arrows_Master, swing1LimitDegrees=10, swing2LimitDegrees=10, twistLimitDegrees=0.1}
 {bone1Name=Dyn_Arrow_4, bone2Name=Dyn_Arrows_Master, swing1LimitDegrees=10, swing2LimitDegrees=10, twistLimitDegrees=0.1}
 {bone1Name=Dyn_Arrow_5, bone2Name=Dyn_Arrows_Master, swing1LimitDegrees=10, swing2LimitDegrees=10, twistLimitDegrees=0.1}
 {bone1Name=Dyn_Arrow_6, bone2Name=Dyn_Arrows_Master, swing1LimitDegrees=10, swing2LimitDegrees=10, twistLimitDegrees=0.1}
 {bone1Name=Dyn_Arrow_7, bone2Name=Dyn_Arrows_Master, swing1LimitDegrees=10, swing2LimitDegrees=10, twistLimitDegrees=0.1}
 {bone1Name=Dyn_Arrow_8, bone2Name=Dyn_Arrows_Master, swing1LimitDegrees=10, swing2LimitDegrees=10, twistLimitDegrees=0.1}
 {bone1Name=Dyn_LongStrap3, bone2Name=Dyn_LongStrap2, swing1LimitDegrees=10, swing2LimitDegrees=20, twistLimitDegrees=1}
 {bone1Name=Dyn_LongStrap2, bone2Name=Dyn_LongStrap1, swing1LimitDegrees=5, swing2LimitDegrees=30, twistLimitDegrees=0.5}
 {bone1Name=Dyn_LongStrap1, bone2Name=Dyn_Belt_Strap, swing1LimitDegrees=10, swing2LimitDegrees=10, twistLimitDegrees=1}
 {bone1Name=Dyn_Belt_Strap, bone2Name=Dyn_Master_Strap, swing1LimitDegrees=10, swing2LimitDegrees=10, twistLimitDegrees=0.1}
 {bone1Name=Dyn_ShortStrap_1, bone2Name=Dyn_Belt_Strap, swing1LimitDegrees=5, swing2LimitDegrees=20, twistLimitDegrees=0.5}
 {bone1Name=Dyn_ShortStrap_2, bone2Name=Dyn_ShortStrap_1, swing1LimitDegrees=10, swing2LimitDegrees=15, twistLimitDegrees=1}
 {bone1Name=Dyn_ShortStrap_3, bone2Name=Dyn_ShortStrap_2, swing1LimitDegrees=10, swing2LimitDegrees=15, twistLimitDegrees=1}
]
