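# Anim Blueprint for the Fox Warrior backpack, which adjusts its fit depending on whether the wearer uses the Fox Warrior character.
# On initialize it compares the loadout's character with that item; a match enables a pack bone offset, while other characters get a root offset.
# A RigidBody node between the two bone edits simulates physics in pack-bone space.
# Skydive and swim-sprint events set another root rotation, which each update overwrites.
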AnimBlueprint /BRCosmetics/Accessories/FORT_Backpacks/Backpack_F_MED_Fox_Warrior/Meshes/F_MED_Fox_Warrior_Pack_AnimBP
target_skeleton=&/F_MED_Fox_Warrior_Pack_Skeleton
graphs [
 AnimGraph {nodes [
   n1 = "Transform (Modify) Bone - Bone: pack"(
    ComponentPose=Local To Component(LocalPose=Local Space Ref Pose())
    Alpha=SelectFloat(A=1, B=0, bPickA=Get isFoxWarriorActive() -> (isFoxWarriorActive:bool))
    Translation="0,-1,2"
    Rotation="0,20,6"
   ) {
    rotationMode=BMM_Replace
    translationMode=BMM_Additive
    alphaBoolBlend {blendOption=Linear}
    boneToModify {boneName=pack}
    rotation {roll=6, yaw=20}
    translation {y=-1, z=2}
   }
   n2 = RigidBody(ComponentPose=@n1) {
    bClampLinearTranslationLimitToRefPose=true
    bFreezeIncomingPoseOnStart=true
    bOverrideWorldGravity=true
    overridePhysicsAsset=&/F_MED_Fox_Warrior_Pack_Physics
    simulationSpace=BaseBoneSpace
    alphaBoolBlend {blendOption=Linear}
    baseBoneRef {boneName=pack}
    componentAppliedLinearAccClamp=<5000 5000 5000>
    componentLinearAccScale=<0.1 0.1 0.1>
    componentLinearVelScale=<0.1 0.1 0.1>
    overrideWorldGravity {z=-400}
   }
   n3 = "Transform (Modify) Bone - Bone: root"(ComponentPose=@n2, Translation=Get FoxWarriorInactive() -> (FoxWarriorInactive:Vector), Rotation=Get RotateRoot() -> (RotateRoot:Rotator)) {
    rotationMode=BMM_Additive
    translationMode=BMM_Replace
    alphaBoolBlend {blendInTime=2, blendOption=Linear, blendOutTime=0.5}
    alphaScaleBiasClamp {bInterpResult=true}
    boneToModify {boneName=root}
   }
   Output Pose(Result=Component To Local(ComponentPose=@n3)) {layerGroup=None}
  ]}
 EventGraph {nodes [
   n1 = Event OnBeginSkydiving()
   Set RotateRoot(execute=[@n1.then, @n6.then, @n7.then], RotateRoot:Rotator="-15, 0,25") -> (Output_Get:Rotator)
   Event BlueprintUpdateAnimation()
   > n2 = Sequence()
   n3 = Get isFoxWarriorActive()
   > Set FoxWarriorInactive(FoxWarriorInactive:Vector=SelectVector(A="0, 0, 0", B="-1.1,-1.1,5", bPickA=@n3)) -> (Output_Get:Vector)
   Set RotateRoot(execute=@n2.then_1, RotateRoot=SelectRotator!(A:Rotator="0, 0, 0", B:Rotator="-12, 0,7", bPickA:bool=@n3) -> (ReturnValue:Rotator))
   Event BlueprintInitializeAnimation()
   > n4 = Cast To FortPlayerPawn(Object=GetOwningActor())
   > n5 = Set currentCID(currentCID:Athena Character Item Definition Object Reference=GetItemDefinitionTypeFromLoadoutWithActor!(
     Actor:Actor Object Reference=@n4
     LoadoutSlot:Gameplay Tag Structure=(TagName="LoadoutSlot.Character")
     CosmeticType:Athena Cosmetic Item Definition Class Reference=/Script/FortniteGame.AthenaCharacterItemDefinition
    ) -> (OutCosmeticItemDef:Athena Character Item Definition Object Reference)) -> (Output_Get:Athena Character Item Definition Object Reference)
   > Set isFoxWarriorActive(isFoxWarriorActive:bool="Equal (Object)"(A=@n5, B=Get foxWarrior() -> (foxWarrior:Athena Character Item Definition Object Reference)) {is=title K2Node_CallFunction}) -> (Output_Get:bool)
   n6 = AnimNotify_SwimSprint_Entered_Layer() {is=title K2Node_Event}
   n7 = AnimNotify_SwimSprintStart_Entered_Layer() {is=title K2Node_Event}
   TryGetPawnOwner()
  ]}
]
variables [
 currentCID {type=AthenaCharacterItemDefinition, default=null}
 foxWarrior {type=AthenaCharacterItemDefinition, default=/BRCosmetics/Athena/Items/Cosmetics/Characters/CID_A_009_Athena_Commando_F_FoxWarrior_21B9R}
 FoxWarriorInactive {type=Vector}
 isFoxWarriorActive {type=boolean}
 RotateRoot {type=Rotator}
]
